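# Shared widget for screen-tool utilities that shows one read-only readout line as a label and value pair.
# An overlay holds a padded horizontal box with two small grey CommonText blocks, the label on the left with a minimum width and the value beside it.
# SetLine takes Label and Value text and writes them into the two blocks; the event graph is empty.
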WidgetBlueprint /Game/Blueprints/Blutilities/ToolsWrkshp/ScrTools/Shared/SharedWidgets/BP_UW_ReadoutLine
graphs [
 EventGraph {nodes [
   Event PreConstruct()
   Event Construct()
   Event Tick()
  ]}
 SetLine {nodes [
   n1 = SetLine!() -> (then, Label:Text, Value:Text) {is=title K2Node_FunctionEntry}
   > "SetText (Text)"(self=Get Value() -> (Value:Common Text Object Reference), InText=@n1.Value)
   > "SetText (Text)"(self=Get LabelEntry() -> (LabelEntry:Common Text Object Reference), InText=@n1)
  ]}
]
variables []
widgets {
 info {inheritedWidgetCount=0, parentClass=&UserWidget, rootWidgetClass=&Overlay, widgetCount=4}
 widgets [
  Overlay_1 {widgetClassPath=&Overlay, widget=Overlay_1}
  HorizontalBox_23 {
   widgetClassPath=&HorizontalBox
   parent=Overlay_1
   widget=HorizontalBox_23
   slot {class=OverlaySlot, properties {horizontalAlignment=HAlign_Fill, verticalAlignment=VAlign_Fill, padding=<ltrb 16 6 0 6>}}
  }
  LabelEntry {
   widgetClassPath=&CommonTextBlock
   bIsVariable=true
   parent=HorizontalBox_23
   widget=LabelEntry
   properties {
    minDesiredWidth=150
    text=Label
    toolTipText=Chudly Text
    colorAndOpacity {colorUseRule=UseColor_Specified, specifiedColor=<rgba 0.5271 0.5271 0.5271 1>}
    font {
     bForceMonospaced=false
     bMaterialIsStencil=false
     fontMaterial=null
     fontObject=/Engine/EditorResources/DefaultEditorFont
     letterSpacing=0
     monospacedWidth=1
     size=7.88
     skewAmount=0
     typefaceFontName=Regular
     outlineSettings {
      bApplyOutlineToDropShadows=false
      bMiteredCorners=false
      bSeparateFillAlpha=false
      outlineMaterial=null
      outlineSize=0
      outlineColor=<rgba 0 0 0 1>
     }
    }
    shadowOffset=<0 0>
   }
   slot {class=HorizontalBoxSlot, properties {horizontalAlignment=HAlign_Left, padding=<ltrb 0 0 10 0>}}
  }
  Value {
   widgetClassPath=&CommonTextBlock
   bIsVariable=true
   parent=HorizontalBox_23
   widget=Value
   properties {
    text=Value
    colorAndOpacity {colorUseRule=UseColor_Specified, specifiedColor=<rgba 0.5271 0.5271 0.5271 1>}
    font {
     bForceMonospaced=false
     bMaterialIsStencil=false
     fontMaterial=null
     fontObject=&DefaultEditorFont
     letterSpacing=0
     monospacedWidth=1
     size=7.88
     skewAmount=0
     typefaceFontName=Regular
     outlineSettings {
      bApplyOutlineToDropShadows=false
      bMiteredCorners=false
      bSeparateFillAlpha=false
      outlineMaterial=null
      outlineSize=0
      outlineColor=<rgba 0 0 0 1>
     }
    }
    shadowOffset=<0 0>
   }
   slot {class=HorizontalBoxSlot, properties {horizontalAlignment=HAlign_Right}}
  }
 ]
}
