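# A single-line row widget shared by the workshop tool screens, showing a grey wrapped text value beside an optional Fix Me button.
# SetLine sets the text and rebuilds its font to switch the typeface between BoldItalic and Regular from the Bold flag.
# It then either shows the button with the supplied tooltip or collapses it.
# Button presses are relayed through the FixButtonClicked dispatcher, leaving the fix to whatever binds it.
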WidgetBlueprint /Game/Blueprints/Blutilities/ToolsWrkshp/ScrTools/Shared/SharedWidgets/BP_UW_SoloLine
graphs [
 EventGraph {nodes [
   "Fix Me Button Pressed (BP_UW_Shared_FixMeButton)"() {is=title K2Node_ComponentBoundEvent}
   > Call Fix Button Clicked!(execute:Exec, self:BP UW Solo Line Object Reference) -> (then:Exec) {is=title K2Node_CallDelegate}
   Event PreConstruct()
   Event Construct()
   Event Tick()
  ]}
 FixButtonClicked {nodes [FixButtonClicked() {is=title K2Node_FunctionEntry}]}
 SetLine {nodes [
   n1 = SetLine!() -> (then, Value:Text, ToolTipText:Text, ButtonVisiblity:bool=~, Bold:bool=~) {is=title K2Node_FunctionEntry}
   n2 = Get Value() -> (Value:Common Text Object Reference)
   > "SetText (Text)"(self=@n2, InText=@n1)
   > n3 = Set TempFont(TempFont:Slate Font Info Structure=Get Font!(self:Text Object Reference=@n2) -> (Font:Slate Font Info Structure)) -> (Output_Get:Slate Font Info Structure)
   n4 = Break Slate Font Info!(SlateFontInfo:"Slate Font Info Structure (by ref)"=@n3) -> (
    FontObject:Object Reference
    FontMaterial:Object Reference
    OutlineSettings:Font Outline Settings Structure
    TypefaceFontName:Name
    Size:float=~
    LetterSpacing:int
    SkewAmount:float=~
    bForceMonospaced:bool=~
    bMaterialIsStencil:bool=~
    MonospacedWidth:float=~
   ) {is=title K2Node_BreakStruct}
   > SetFont(
    self=Get Value()
    InFontInfo=MakeSlateFontInfo!(
     FontObject:Object Reference=@n4
     FontMaterial:Object Reference=@n4.FontMaterial
     OutlineSettings:Font Outline Settings Structure=@n4.OutlineSettings
     TypefaceFontName:Name=SelectName!(A:Name=BoldItalic, B:Name=Regular, bPickA:bool=@n1.Bold) -> (ReturnValue:Name=None)
     Size:float=@n4.Size
     LetterSpacing:int=@n4.LetterSpacing
     SkewAmount:float=@n4.SkewAmount
     bForceMonospaced:bool=@n4.bForceMonospaced
     bMaterialIsStencil:bool=@n4.bMaterialIsStencil
     MonospacedWidth:float=@n4.MonospacedWidth
    ) -> (ReturnValue:Slate Font Info Structure)
   )
   > n5 = Branch(Condition=@n1.ButtonVisiblity)
   n6 = Get BP_UW_Shared_FixMeButton() -> (BP_UW_Shared_FixMeButton:BP UW Shared Fix Me Button Object Reference)
   > Set Visibility(Visibility=Visible, self=@n6)
   > SetToolTipText!(execute, self:Widget Object Reference=@n6, InToolTipText:"Text (by ref)"=Get ToolTipText() -> (ToolTipText:Text)) -> (then)
   Set Visibility(execute=@n5.else, Visibility=Collapsed, self=@n6)
  ]}
]
variables [FixButtonClicked {}]
widgets {
 info {inheritedWidgetCount=0, parentClass=&UserWidget, rootWidgetClass=&Overlay, widgetCount=4}
 widgets [
  Overlay_1 {widgetClassPath=&Overlay, widget=Overlay_1}
  HorizontalBox_23 {
   widgetClassPath=&HorizontalBox
   parent=Overlay_1
   widget=HorizontalBox_23
   slot {class=OverlaySlot, properties {horizontalAlignment=HAlign_Fill, verticalAlignment=VAlign_Fill, padding=<ltrb 16 6 0 6>}}
  }
  Value {
   widgetClassPath=&CommonTextBlock
   bIsVariable=true
   parent=HorizontalBox_23
   widget=Value
   properties {
    autoWrapText=true
    minDesiredWidth=300
    text=Value
    wrapTextAt=300
    colorAndOpacity {colorUseRule=UseColor_Specified, specifiedColor=<rgba 0.5271 0.5271 0.5271 1>}
    font {
     bForceMonospaced=false
     bMaterialIsStencil=false
     fontMaterial=null
     fontObject=/Engine/EditorResources/DefaultEditorFont
     letterSpacing=0
     monospacedWidth=1
     size=7.88
     skewAmount=0
     typefaceFontName=Regular
     outlineSettings {
      bApplyOutlineToDropShadows=false
      bMiteredCorners=false
      bSeparateFillAlpha=false
      outlineMaterial=null
      outlineSize=0
      outlineColor=<rgba 0 0 0 1>
     }
    }
    shadowOffset=<0 0>
   }
   slot {class=HorizontalBoxSlot, properties {horizontalAlignment=HAlign_Left, padding=<ltrb 0 0 10 0>}}
  }
  BP_UW_Shared_FixMeButton {
   widgetClassPath=&/BP_UW_Shared_FixMeButton
   bIsVariable=true
   parent=HorizontalBox_23
   widget=BP_UW_Shared_FixMeButton
   slot {
    class=HorizontalBoxSlot
    properties {horizontalAlignment=HAlign_Center, verticalAlignment=VAlign_Center, padding=<ltrb 0 -6 0 -6>, size {sizeRule=Fill, value=1}}
   }
  }
 ]
}
