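# Shared tool-UI row widget showing a label, a value and a right-aligned status indicator.
# SetLine fills both texts and shows one overlaid indicator: a green check without MoreInfo, otherwise the Fix Me button if a fix is available or an info icon if not.
# The tooltip text goes on whichever indicator is visible.
# Fix Me presses are relayed through the FixButtonClicked dispatcher, leaving the fix to listeners.
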WidgetBlueprint /Game/Blueprints/Blutilities/ToolsWrkshp/ScrTools/Shared/SharedWidgets/BP_UW_TripleLine
graphs [
 EventGraph {nodes [
   "Fix Me Button Pressed (BP_UW_Shared_FixMeButton)"() {is=title K2Node_ComponentBoundEvent}
   > Call Fix Button Clicked!(execute:Exec, self:BP UW Triple Line Object Reference) -> (then:Exec) {is=title K2Node_CallDelegate}
   Event PreConstruct()
   Event Construct()
   Event Tick()
  ]}
 FixButtonClicked {nodes [FixButtonClicked() {is=title K2Node_FunctionEntry}]}
 SetLine {nodes [
   n1 = SetLine!() -> (then, Label:Text, Value:Text, ToolTipText:Text, MoreInfo:bool=~, FixAvaliable:bool=~) {is=title K2Node_FunctionEntry}
   > "SetText (Text)"(self=Get LabelEntry() -> (LabelEntry:Common Text Object Reference), InText=@n1)
   > "SetText (Text)"(self=Get Value() -> (Value:Common Text Object Reference), InText=@n1.Value)
   > Set NewToolTipText(NewToolTipText:Text=@n1.ToolTipText) -> (Output_Get:Text)
   > n2 = Branch(Condition=@n1.MoreInfo)
   n3 = Get check-circle() -> (check-circle:Image Object Reference)
   > Set Visibility(Visibility=Collapsed, self=@n3)
   > n4 = Branch(Condition=@n1.FixAvaliable)
   n5 = Get BP_UW_Shared_FixMeButton() -> (BP_UW_Shared_FixMeButton:BP UW Shared Fix Me Button Object Reference)
   > Set Visibility(Visibility=Visible, self=@n5)
   n6 = Get info-circle() -> (info-circle:Image Object Reference)
   > Set Visibility(Visibility=Collapsed, self=@n6)
   n7 = Get NewToolTipText() -> (NewToolTipText:Text)
   > SetToolTipText!(execute, self:Widget Object Reference=Get BP_UW_Shared_FixMeButton(), InToolTipText:"Text (by ref)"=@n7) -> (then)
   Set Visibility(execute=@n4.else, Visibility=Collapsed, self=@n5)
   > Set Visibility(Visibility=Visible, self=@n6)
   > SetToolTipText(self=Get info-circle(), InToolTipText=@n7)
   Set Visibility(execute=@n2.else, Visibility=Visible, self=@n3)
   > Set Visibility(Visibility=Collapsed, self=Get info-circle())
   > Set Visibility(Visibility=Collapsed, self=Get BP_UW_Shared_FixMeButton())
   > SetToolTipText(self=@n3, InToolTipText=Get NewToolTipText())
  ]}
]
variables [FixButtonClicked {}]
widgets {
 info {inheritedWidgetCount=0, parentClass=&UserWidget, rootWidgetClass=&Overlay, widgetCount=8}
 widgets [
  Overlay_1 {widgetClassPath=&Overlay, widget=Overlay_1}
  HorizontalBox_23 {
   widgetClassPath=&HorizontalBox
   parent=Overlay_1
   widget=HorizontalBox_23
   slot {class=OverlaySlot, properties {horizontalAlignment=HAlign_Fill, verticalAlignment=VAlign_Fill, padding=<ltrb 16 6 0 6>}}
  }
  LabelEntry {
   widgetClassPath=&CommonTextBlock
   bIsVariable=true
   parent=HorizontalBox_23
   widget=LabelEntry
   properties {
    autoWrapText=true
    minDesiredWidth=150
    text=Label
    toolTipText=Chudly Text
    wrapTextAt=150
    colorAndOpacity {colorUseRule=UseColor_Specified, specifiedColor=<rgba 0.5271 0.5271 0.5271 1>}
    font {
     bForceMonospaced=false
     bMaterialIsStencil=false
     fontMaterial=null
     fontObject=/Engine/EditorResources/DefaultEditorFont
     letterSpacing=0
     monospacedWidth=1
     size=7.88
     skewAmount=0
     typefaceFontName=Regular
     outlineSettings {
      bApplyOutlineToDropShadows=false
      bMiteredCorners=false
      bSeparateFillAlpha=false
      outlineMaterial=null
      outlineSize=0
      outlineColor=<rgba 0 0 0 1>
     }
    }
    shadowOffset=<0 0>
   }
   slot {class=HorizontalBoxSlot, properties {horizontalAlignment=HAlign_Left, verticalAlignment=VAlign_Center, padding=<ltrb 0 0 10 0>}}
  }
  Value {
   widgetClassPath=&CommonTextBlock
   bIsVariable=true
   parent=HorizontalBox_23
   widget=Value
   properties {
    autoWrapText=true
    minDesiredWidth=150
    text=Value
    wrapTextAt=150
    colorAndOpacity {colorUseRule=UseColor_Specified, specifiedColor=<rgba 0.5271 0.5271 0.5271 1>}
    font {
     bForceMonospaced=false
     bMaterialIsStencil=false
     fontMaterial=null
     fontObject=&DefaultEditorFont
     letterSpacing=0
     monospacedWidth=1
     size=7.88
     skewAmount=0
     typefaceFontName=Regular
     outlineSettings {
      bApplyOutlineToDropShadows=false
      bMiteredCorners=false
      bSeparateFillAlpha=false
      outlineMaterial=null
      outlineSize=0
      outlineColor=<rgba 0 0 0 1>
     }
    }
    shadowOffset=<0 0>
   }
   slot {class=HorizontalBoxSlot, properties {horizontalAlignment=HAlign_Left, verticalAlignment=VAlign_Center, padding=<ltrb 0 0 10 0>}}
  }
  Overlay_2 {
   widgetClassPath=&Overlay
   parent=HorizontalBox_23
   widget=Overlay_2
   properties {bOverride_Cursor=true}
   slot {class=HorizontalBoxSlot, properties {horizontalAlignment=HAlign_Center, verticalAlignment=VAlign_Center, size {sizeRule=Fill, value=1}}}
  }
  BP_UW_Shared_FixMeButton {
   widgetClassPath=&/BP_UW_Shared_FixMeButton
   bIsVariable=true
   parent=Overlay_2
   widget=BP_UW_Shared_FixMeButton
   slot {class=OverlaySlot, properties {horizontalAlignment=HAlign_Right, verticalAlignment=VAlign_Center, padding=<ltrb 0 -6 0 -6>}}
  }
  check-circle {
   widgetClassPath=&Image
   bIsVariable=true
   parent=Overlay_2
   widget=check-circle
   properties {
    brush {
     bIsDynamicallyLoaded=false
     drawAs=Image
     imageType=FullColor
     mirroring=NoMirror
     resourceName=None
     resourceObject=/Game/Blueprints/Blutilities/ToolsWrkshp/ScrTools/Shared/SharedWidgets/graphics/checkcircle
     tiling=NoTile
     imageSize=<16 16>
     margin=<ltrb 0 0 0 0>
     outlineSettings {
      bUseBrushTransparency=false
      roundingType=HalfHeightRadius
      width=0
      color {colorUseRule=UseColor_Specified, specifiedColor=<rgba 0 0 0 0>}
      cornerRadii=<0 0 0 0>
     }
     tintColor {colorUseRule=UseColor_Specified, specifiedColor=<rgba 1 1 1 1>}
     uVRegion {bIsValid=false, max=<0 0>, min=<0 0>}
    }
    colorAndOpacity=<rgba 0.0137 0.7758 0.0704 1>
   }
   slot {class=OverlaySlot, properties {horizontalAlignment=HAlign_Right, verticalAlignment=VAlign_Center, padding=<ltrb 0 -1 0 -1>}}
  }
  info-circle {
   widgetClassPath=&Image
   bIsVariable=true
   parent=Overlay_2
   widget=info-circle
   properties {
    brush {
     bIsDynamicallyLoaded=false
     drawAs=Image
     imageType=FullColor
     mirroring=NoMirror
     resourceName=None
     resourceObject=/Game/Blueprints/Blutilities/ToolsWrkshp/ScrTools/Shared/SharedWidgets/graphics/infocircle
     tiling=NoTile
     imageSize=<16 16>
     margin=<ltrb 0 0 0 0>
     outlineSettings {
      bUseBrushTransparency=false
      roundingType=HalfHeightRadius
      width=0
      color {colorUseRule=UseColor_Specified, specifiedColor=<rgba 0 0 0 0>}
      cornerRadii=<0 0 0 0>
     }
     tintColor {colorUseRule=UseColor_Specified, specifiedColor=<rgba 1 1 1 1>}
     uVRegion {bIsValid=false, max=<0 0>, min=<0 0>}
    }
    colorAndOpacity=<rgba 0.5271 0.5271 0.5271 1>
   }
   slot {class=OverlaySlot, properties {horizontalAlignment=HAlign_Right, verticalAlignment=VAlign_Center, padding=<ltrb 0 -1 0 -1>}}
  }
 ]
}
