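# Shared tools widget providing a titled, collapsible box with dynamically added readout rows.
# A caret-styled CheckBox in the header drives SetBoxState, toggling the BoxContent named slot between Visible and Collapsed.
# Outside design time, PreConstruct calls SetUpBox to apply name, size, color and initial open state.
# Helper functions lazily create Readout, Solo or Triple lines in DynamicEntries when none exists, and AddShortcuts builds rows from a string map.
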WidgetBlueprint /Game/Blueprints/Blutilities/ToolsWrkshp/ScrTools/Shared/SharedWidgets/ExpandableBox/BP_UW_ExpandableBox
graphs [
 AddShortcuts {nodes [
   n1 = Format Text(Format="LOCTEXT(\"Tool Tip Text\")")
   n2 = Construct Object from Class(
    execute=[@n7.else, @n8.Is Not Valid]
    Class=/Game/Blueprints/Blutilities/ToolsWrkshp/ScrTools/Shared/SharedWidgets/BP_UW_ReadoutLine.BP_UW_ReadoutLine_C
   ) -> (ReturnValue:BP UW Readout Line Object Reference)
   n3 = Get LabelEntry!(self:BP UW Readout Line Object Reference=@n2) -> (LabelEntry:Common Text Object Reference)
   SetToolTipText!(execute, self:Widget Object Reference=@n3, InToolTipText:"Text (by ref)"=@n1) -> (then)
   n4 = Get Value!(self:BP UW Readout Line Object Reference=@n2) -> (Value:Common Text Object Reference)
   > SetToolTipText(self=@n4, InToolTipText=@n1)
   n5 = AddShortcuts!() -> (
    then
    ShortcutMap:Map of Strings to Strings
    ExistingShortcutLines:Array of BP UW Readout Line Object References
    ShortcutMapKeys:Array of Strings
   ) {is=title K2Node_FunctionEntry}
   > Set ShortcutIndex(ShortcutIndex:int=0) -> (Output_Get:int=0)
   > Set ShortcutLines(ShortcutLines:Array of BP UW Readout Line Object References=@n5.ExistingShortcutLines) -> (Output_Get:Array of BP UW Readout Line Object References)
   > n6 = For Each Loop(Array=@n5.ShortcutMapKeys)
   Return Node!(
    execute=@n6.Completed
    AllShortcutLines:Array of BP UW Readout Line Object References=Get ShortcutLines() -> (ShortcutLines:Array of BP UW Readout Line Object References)
   )
   > Set MapKeyInLoop(MapKeyInLoop:String=@n6) -> (Output_Get:String)
   > Set MapValueInLoop(MapValueInLoop:String=Find(TargetMap=@n5, Key=@n6)) -> (Output_Get:String)
   > n7 = Branch(Condition=Is Valid Index!(TargetArray:Array of BP UW Readout Line Object References=Get ShortcutLines(), IndexToTest:int=@n6.Array Index) -> (ReturnValue:bool))
   > n8 = Is Valid(InputObject="Get (a copy)"(Array:Array of BP UW Readout Line Object References=Get ShortcutLines(), Dimension 1=Get ShortcutIndex() -> (ShortcutIndex:int=0)) -> (Output:BP UW Readout Line Object Reference))
   Increment Int(" "=[@n9.then, @n8.Is Valid], Value=Get ShortcutIndex() -> (ShortcutIndex=0))
   Set ToolTipText!(execute, ToolTipText:Text=@n1, self:Widget Object Reference=@n4) -> (then, Output_Get:Text)
   > Set ToolTipText(ToolTipText=@n1, self=@n3)
   > Set ToolTipText(ToolTipText=@n1, self=@n2)
   "For Each Loop (Map)"!(execute:Exec, MapPin:Map of Strings to Strings=@n5, BreakPin:Exec) -> (then:Exec, KeyPin:String, ValuePin:String, CompletedPin:Exec)
   Get ShortcutIndex() -> (ShortcutIndex=0)
   Add(
    execute=@n10.then
    TargetArray:Array of BP UW Readout Line Object References=Get ShortcutLines()
    NewItem:"BP UW Readout Line Object Reference (by ref)"=@n2
   ) {is=title K2Node_CallArrayFunction}
   > "SetText (Text)"(self=@n3, InText="To Text (String)"(InString=Get MapKeyInLoop() -> (MapKeyInLoop:String)))
   > n9 = "SetText (Text)"(self=@n4, InText="To Text (String)"(InString=Get MapValueInLoop() -> (MapValueInLoop:String)))
   n10 = AddChildToVerticalBox(execute=@n2.then, self=Get DynamicEntries(self=Self-Reference()), Content=@n2)
  ]}
 EventGraph {nodes [
   n1 = "On Check State Changed (ExpandButton)"!() -> (OutputDelegate, then, bIsChecked:bool) {is=title K2Node_ComponentBoundEvent}
   > SetBoxState!(execute, self, Checked:bool=@n1.bIsChecked) -> (then)
   n2 = Event PreConstruct()
   > n3 = Branch(Condition=@n2.IsDesignTime)
   SetUpBox(
    execute=@n3.else
    self:Self Object Reference
    Name=Get Name() -> (Name:Text)
    BoxSize=Get BoxSize() -> (BoxSize:Vector 2D Structure)
    OpenByDefault=Get OpenByDefault() -> (OpenByDefault:bool)
   )
   Event Construct()
   Event Tick()
  ]}
 SetBoxState {nodes [
   n1 = SetBoxState!() -> (then, Checked:bool=~) {is=title K2Node_FunctionEntry}
   > n2 = Branch(Condition=@n1)
   > SetCheckedState!(
    execute
    self:Check Box Object Reference=Get ExpandButton() -> (ExpandButton:Check Box Object Reference)
    InCheckedState:ECheckBoxState Enum=Checked
   ) -> (then)
   > SetVisibility(self=Get BoxContent() -> (BoxContent:Named Slot Object Reference), InVisibility=Visible)
   SetCheckedState(execute=@n2.else, self=Get ExpandButton(), InCheckedState=Unchecked)
   > SetVisibility(self=Get BoxContent(), InVisibility=Collapsed)
  ]}
 SetUpBox {nodes [
   n1 = SetUpBox!() -> (
    then
    Name:Text
    BoxSize:Vector 2D Structure=(X=300,Y=25)
    BoxColor:Linear Color Structure=(R=0.0284,G=0.0284,B=0.0284,A=1)
    OpenByDefault:bool=true
   ) {is=title K2Node_FunctionEntry}
   > Set Name(Name:Text=@n1) -> (Output_Get:Text)
   > Set BoxSize(BoxSize:Vector 2D Structure=@n1.BoxSize) -> (Output_Get:Vector 2D Structure)
   > Set BoxColor(BoxColor:Linear Color Structure=@n1.BoxColor) -> (Output_Get:Linear Color Structure)
   > Set OpenByDefault(OpenByDefault:bool=@n1.OpenByDefault) -> (Output_Get:bool)
   > "SetText (Text)"(self=Get BoxName() -> (BoxName:Text Object Reference), InText=Get Name())
   > SetWidthOverride(self=Get Size() -> (Size:Size Box Object Reference), InWidthOverride=Get!() -> (BoxSize_X:double, BoxSize_Y:double))
   > SetHeightOverride(self=Get Size(), InHeightOverride=Get().BoxSize_Y)
   > SetBrushColor(self=Get Border() -> (Border:Border Object Reference), InBrushColor=Get BoxColor() -> (BoxColor:Linear Color Structure))
   > SetBoxState(Checked=Get OpenByDefault()) {is=title K2Node_CallFunction}
  ]}
 SetupLine {nodes [
   n1 = SetupLine!() -> (then, CurrentLine:BP UW Readout Line Object Reference, DynamicLine:Vertical Box Object Reference, Label:Text, Value:Text) {is=title K2Node_FunctionEntry}
   > n2 = Is Valid(InputObject=@n1)
   n3 = Construct Object from Class(execute=@n2.Is Not Valid, Class=&BP_UW_ReadoutLine_C)
   > n4 = Set NewReadoutLine(NewReadoutLine:BP UW Readout Line Object Reference=@n3) -> (Output_Get:BP UW Readout Line Object Reference)
   > AddChildToVerticalBox(self=Get DynamicEntries!() -> (DynamicEntries), Content=@n4)
   > SetLine!(
    execute
    self:BP UW Readout Line Object Reference=Get NewReadoutLine() -> (NewReadoutLine:BP UW Readout Line Object Reference)
    Label:Text=Get Label() -> (Label:Text)
    Value:Text=Get Value!() -> (Value:Text)
   ) -> (then)
   > Return Node!(execute, Readout Line:BP UW Readout Line Object Reference=Get NewReadoutLine())
   Return Node(execute=@n2.Is Valid, Readout Line=Get CurrentLine() -> (CurrentLine:BP UW Readout Line Object Reference))
  ]}
 SetupLineSolo {nodes [
   n1 = SetupLineSolo!() -> (
    then
    Current Line:BP UW Solo Line Object Reference
    Dynamic Line:Vertical Box Object Reference
    Value:Text
    ToolTipText:Text
    Button Visible:bool=~
    Bold:bool=~
   ) {is=title K2Node_FunctionEntry}
   > n2 = Is Valid(InputObject=@n1)
   Is Valid(exec=@n2.Is Not Valid, InputObject=Get DynamicEntries())
   > n3 = Construct Object from Class(Class=/Game/Blueprints/Blutilities/ToolsWrkshp/ScrTools/Shared/SharedWidgets/BP_UW_SoloLine.BP_UW_SoloLine_C) -> (ReturnValue:BP UW Solo Line Object Reference)
   > n4 = Set NewReadoutLine(NewReadoutLine:BP UW Solo Line Object Reference=@n3) -> (Output_Get:BP UW Solo Line Object Reference)
   > AddChildToVerticalBox(self=Get DynamicEntries(), Content=@n4)
   n5 = Get NewReadoutLine() -> (NewReadoutLine:BP UW Solo Line Object Reference)
   > SetLine!(
    execute
    self:BP UW Solo Line Object Reference=@n5
    Value=@n1.Value
    ToolTipText:Text=Get ToolTipText() -> (ToolTipText:Text)
    ButtonVisiblity:bool=@n1.Button Visible
    Bold:bool=@n1.Bold
   ) -> (then)
   > Return Node(Readout Line:BP UW Solo Line Object Reference=@n5)
   Return Node(execute=@n2.Is Valid, Readout Line=Get Current Line() -> (Current Line:BP UW Solo Line Object Reference))
  ]}
 SetupLineTriple {nodes [
   n1 = SetupLineTriple!() -> (
    then
    Current Line:BP UW Triple Line Object Reference
    Dynamic Line:Vertical Box Object Reference
    Label:Text
    Value:Text
    ToolTipText:Text
    MoreInfo:bool=~
    FixAvaliable:bool=~
   ) {is=title K2Node_FunctionEntry}
   > n2 = Is Valid(InputObject=@n1)
   Is Valid(exec=@n2.Is Not Valid, InputObject=Get DynamicEntries())
   > n3 = Construct Object from Class(Class=/Game/Blueprints/Blutilities/ToolsWrkshp/ScrTools/Shared/SharedWidgets/BP_UW_TripleLine.BP_UW_TripleLine_C) -> (ReturnValue:BP UW Triple Line Object Reference)
   > n4 = Set NewReadoutLine(NewReadoutLine:BP UW Triple Line Object Reference=@n3) -> (Output_Get:BP UW Triple Line Object Reference)
   > AddChildToVerticalBox(self=Get DynamicEntries(), Content=@n4)
   n5 = Get NewReadoutLine() -> (NewReadoutLine:BP UW Triple Line Object Reference)
   > SetLine!(
    execute
    self:BP UW Triple Line Object Reference=@n5
    Label:Text=Get Label()
    Value=Get Value()
    ToolTipText=Get ToolTipText()
    MoreInfo:bool=Get MoreInfo() -> (MoreInfo:bool)
    FixAvaliable:bool=Get FixAvaliable() -> (FixAvaliable:bool)
   ) -> (then)
   > Return Node(Readout Line:BP UW Triple Line Object Reference=@n5)
   Return Node(execute=@n2.Is Valid, Readout Line=Get Current Line() -> (Current Line:BP UW Triple Line Object Reference))
  ]}
]
variables [
 BoxColor {type=LinearColor, default=<rgba 0.0284 0.0284 0.0284 1>}
 BoxSize {type=Vector2D, default=<300 25>}
 Name {type=string, default=""}
 OpenByDefault {type=boolean}
]
widgets {
 info {inheritedWidgetCount=0, parentClass=&UserWidget, rootWidgetClass=&VerticalBox, widgetCount=7}
 widgets [
  ExpandableBox {widgetClassPath=&VerticalBox, widget=ExpandableBox}
  Size {
   widgetClassPath=&SizeBox
   bIsVariable=true
   parent=ExpandableBox
   widget=Size
   properties {bOverride_HeightOverride=true, bOverride_WidthOverride=true}
   slot {class=VerticalBoxSlot, properties {padding=<ltrb 0 4 0 4>}}
  }
  Border {
   widgetClassPath=&Border
   bIsVariable=true
   parent=Size
   widget=Border
   properties {
    background {
     bIsDynamicallyLoaded=false
     drawAs=Box
     imageType=NoImage
     mirroring=NoMirror
     resourceName=None
     resourceObject=null
     tiling=NoTile
     imageSize=<32 32>
     margin=<ltrb 0 0 0 0>
     outlineSettings {
      bUseBrushTransparency=false
      roundingType=HalfHeightRadius
      width=0
      color {colorUseRule=UseColor_Specified, specifiedColor=<rgba 0 0 0 0>}
      cornerRadii=<0 0 0 0>
     }
     tintColor {colorUseRule=UseColor_Specified, specifiedColor=<rgba 1 1 1 1>}
     uVRegion {bIsValid=false, max=<0 0>, min=<0 0>}
    }
    contentColorAndOpacity=<rgba 0.5 0.5 0.5 1>
    padding=<ltrb 2 0 0 0>
   }
   slot {class=SizeBoxSlot}
  }
  ExpandButton {
   widgetClassPath=/Script/UMG.CheckBox
   bIsVariable=true
   parent=Border
   widget=ExpandButton
   properties {
    checkedState=Checked
    clickMethod=MouseDown
    pressMethod=ButtonPress
    touchMethod=Down
    widgetStyle {
     checkBoxType=CheckBox
     backgroundHoveredImage {
      bIsDynamicallyLoaded=false
      drawAs=NoDrawType
      imageType=NoImage
      mirroring=NoMirror
      resourceName=None
      resourceObject=null
      tiling=NoTile
      imageSize=<0 0>
      margin=<ltrb 0 0 0 0>
      outlineSettings {
       bUseBrushTransparency=false
       roundingType=HalfHeightRadius
       width=0
       color {colorUseRule=UseColor_Specified, specifiedColor=<rgba 0 0 0 0>}
       cornerRadii=<0 0 0 0>
      }
      tintColor {colorUseRule=UseColor_Specified, specifiedColor=<rgba 1 1 1 1>}
      uVRegion {bIsValid=false, max=<0 0>, min=<0 0>}
     }
     backgroundImage {
      bIsDynamicallyLoaded=false
      drawAs=NoDrawType
      imageType=NoImage
      mirroring=NoMirror
      resourceName=None
      resourceObject=null
      tiling=NoTile
      imageSize=<0 0>
      margin=<ltrb 0 0 0 0>
      outlineSettings {
       bUseBrushTransparency=false
       roundingType=HalfHeightRadius
       width=0
       color {colorUseRule=UseColor_Specified, specifiedColor=<rgba 0 0 0 0>}
       cornerRadii=<0 0 0 0>
      }
      tintColor {colorUseRule=UseColor_Specified, specifiedColor=<rgba 1 1 1 1>}
      uVRegion {bIsValid=false, max=<0 0>, min=<0 0>}
     }
     backgroundPressedImage {
      bIsDynamicallyLoaded=false
      drawAs=NoDrawType
      imageType=NoImage
      mirroring=NoMirror
      resourceName=None
      resourceObject=null
      tiling=NoTile
      imageSize=<0 0>
      margin=<ltrb 0 0 0 0>
      outlineSettings {
       bUseBrushTransparency=false
       roundingType=HalfHeightRadius
       width=0
       color {colorUseRule=UseColor_Specified, specifiedColor=<rgba 0 0 0 0>}
       cornerRadii=<0 0 0 0>
      }
      tintColor {colorUseRule=UseColor_Specified, specifiedColor=<rgba 1 1 1 1>}
      uVRegion {bIsValid=false, max=<0 0>, min=<0 0>}
     }
     borderBackgroundColor {colorUseRule=UseColor_Specified, specifiedColor=<rgba 1 1 1 1>}
     checkedForeground {colorUseRule=UseColor_Foreground, specifiedColor=<rgba 1 0 1 1>}
     checkedHoveredForeground {colorUseRule=UseColor_Foreground, specifiedColor=<rgba 1 0 1 1>}
     checkedHoveredImage {
      bIsDynamicallyLoaded=false
      drawAs=Image
      imageType=FullColor
      mirroring=NoMirror
      resourceName=None
      resourceObject=&/caret-down
      tiling=NoTile
      imageSize=<16 16>
      margin=<ltrb 0 0 0 0>
      outlineSettings {
       bUseBrushTransparency=false
       roundingType=HalfHeightRadius
       width=0
       color {colorUseRule=UseColor_Specified, specifiedColor=<rgba 0 0 0 0>}
       cornerRadii=<0 0 0 0>
      }
      tintColor {colorUseRule=UseColor_Specified, specifiedColor=<rgba 0.9167 0.9167 0.9167 1>}
      uVRegion {bIsValid=false, max=<0 0>, min=<0 0>}
     }
     checkedImage {
      bIsDynamicallyLoaded=false
      drawAs=Image
      imageType=FullColor
      mirroring=NoMirror
      resourceName=None
      resourceObject=&caret-down
      tiling=NoTile
      imageSize=<16 16>
      margin=<ltrb 0 0 0 0>
      outlineSettings {
       bUseBrushTransparency=false
       roundingType=HalfHeightRadius
       width=0
       color {colorUseRule=UseColor_Specified, specifiedColor=<rgba 0 0 0 0>}
       cornerRadii=<0 0 0 0>
      }
      tintColor {colorUseRule=UseColor_Specified, specifiedColor=<rgba 0.9167 0.9167 0.9167 1>}
      uVRegion {bIsValid=false, max=<0 0>, min=<0 0>}
     }
     checkedPressedForeground {colorUseRule=UseColor_Foreground, specifiedColor=<rgba 1 0 1 1>}
     checkedPressedImage {
      bIsDynamicallyLoaded=false
      drawAs=Image
      imageType=FullColor
      mirroring=NoMirror
      resourceName=None
      resourceObject=&caret-down
      tiling=NoTile
      imageSize=<16 16>
      margin=<ltrb 0 0 0 0>
      outlineSettings {
       bUseBrushTransparency=false
       roundingType=HalfHeightRadius
       width=0
       color {colorUseRule=UseColor_Specified, specifiedColor=<rgba 0 0 0 0>}
       cornerRadii=<0 0 0 0>
      }
      tintColor {colorUseRule=UseColor_Specified, specifiedColor=<rgba 0.9167 0.9167 0.9167 1>}
      uVRegion {bIsValid=false, max=<0 0>, min=<0 0>}
     }
     checkedSlateSound {resourceObject=null}
     foregroundColor {colorUseRule=UseColor_Foreground, specifiedColor=<rgba 1 0 1 1>}
     hoveredForeground {colorUseRule=UseColor_Foreground, specifiedColor=<rgba 1 0 1 1>}
     hoveredSlateSound {resourceObject=null}
     padding=<ltrb 3 0 0 0>
     pressedForeground {colorUseRule=UseColor_Foreground, specifiedColor=<rgba 1 0 1 1>}
     uncheckedHoveredImage {
      bIsDynamicallyLoaded=false
      drawAs=Image
      imageType=FullColor
      mirroring=NoMirror
      resourceName=None
      resourceObject=&/caret-right
      tiling=NoTile
      imageSize=<16 16>
      margin=<ltrb 0 0 0 0>
      outlineSettings {
       bUseBrushTransparency=false
       roundingType=HalfHeightRadius
       width=0
       color {colorUseRule=UseColor_Specified, specifiedColor=<rgba 0 0 0 0>}
       cornerRadii=<0 0 0 0>
      }
      tintColor {colorUseRule=UseColor_Specified, specifiedColor=<rgba 0.9167 0.9167 0.9167 1>}
      uVRegion {bIsValid=false, max=<0 0>, min=<0 0>}
     }
     uncheckedImage {
      bIsDynamicallyLoaded=false
      drawAs=Image
      imageType=FullColor
      mirroring=NoMirror
      resourceName=None
      resourceObject=&caret-right
      tiling=NoTile
      imageSize=<16 16>
      margin=<ltrb 0 0 0 0>
      outlineSettings {
       bUseBrushTransparency=false
       roundingType=HalfHeightRadius
       width=0
       color {colorUseRule=UseColor_Specified, specifiedColor=<rgba 0 0 0 0>}
       cornerRadii=<0 0 0 0>
      }
      tintColor {colorUseRule=UseColor_Specified, specifiedColor=<rgba 0.9167 0.9167 0.9167 1>}
      uVRegion {bIsValid=false, max=<0 0>, min=<0 0>}
     }
     uncheckedPressedImage {
      bIsDynamicallyLoaded=false
      drawAs=Image
      imageType=FullColor
      mirroring=NoMirror
      resourceName=None
      resourceObject=&caret-right
      tiling=NoTile
      imageSize=<16 16>
      margin=<ltrb 0 0 0 0>
      outlineSettings {
       bUseBrushTransparency=false
       roundingType=HalfHeightRadius
       width=0
       color {colorUseRule=UseColor_Specified, specifiedColor=<rgba 0 0 0 0>}
       cornerRadii=<0 0 0 0>
      }
      tintColor {colorUseRule=UseColor_Specified, specifiedColor=<rgba 0.9167 0.9167 0.9167 1>}
      uVRegion {bIsValid=false, max=<0 0>, min=<0 0>}
     }
     uncheckedSlateSound {resourceObject=null}
     undeterminedForeground {colorUseRule=UseColor_Foreground, specifiedColor=<rgba 1 0 1 1>}
     undeterminedHoveredImage {
      bIsDynamicallyLoaded=false
      drawAs=Image
      imageType=FullColor
      mirroring=NoMirror
      resourceName=../../../Engine/Content/Slate/Common/CheckBox_Undetermined_Hovered.png
      resourceObject=null
      tiling=NoTile
      imageSize=<16 16>
      margin=<ltrb 0 0 0 0>
      outlineSettings {
       bUseBrushTransparency=false
       roundingType=HalfHeightRadius
       width=0
       color {colorUseRule=UseColor_Specified, specifiedColor=<rgba 0 0 0 0>}
       cornerRadii=<0 0 0 0>
      }
      tintColor {colorUseRule=UseColor_Specified, specifiedColor=<rgba 1 1 1 1>}
      uVRegion {bIsValid=false, max=<0 0>, min=<0 0>}
     }
     undeterminedImage {
      bIsDynamicallyLoaded=false
      drawAs=Image
      imageType=FullColor
      mirroring=NoMirror
      resourceName=../../../Engine/Content/Slate/Common/CheckBox_Undetermined.png
      resourceObject=null
      tiling=NoTile
      imageSize=<16 16>
      margin=<ltrb 0 0 0 0>
      outlineSettings {
       bUseBrushTransparency=false
       roundingType=HalfHeightRadius
       width=0
       color {colorUseRule=UseColor_Specified, specifiedColor=<rgba 0 0 0 0>}
       cornerRadii=<0 0 0 0>
      }
      tintColor {colorUseRule=UseColor_Specified, specifiedColor=<rgba 1 1 1 1>}
      uVRegion {bIsValid=false, max=<0 0>, min=<0 0>}
     }
     undeterminedPressedImage {
      bIsDynamicallyLoaded=false
      drawAs=Image
      imageType=FullColor
      mirroring=NoMirror
      resourceName=../../../Engine/Content/Slate/Common/CheckBox_Undetermined_Hovered.png
      resourceObject=null
      tiling=NoTile
      imageSize=<16 16>
      margin=<ltrb 0 0 0 0>
      outlineSettings {
       bUseBrushTransparency=false
       roundingType=HalfHeightRadius
       width=0
       color {colorUseRule=UseColor_Specified, specifiedColor=<rgba 0 0 0 0>}
       cornerRadii=<0 0 0 0>
      }
      tintColor {colorUseRule=UseColor_Specified, specifiedColor=<rgba 0.5 0.5 0.5 1>}
      uVRegion {bIsValid=false, max=<0 0>, min=<0 0>}
     }
    }
   }
   slot {class=BorderSlot, properties {padding=<ltrb 2 0 0 0>}}
  }
  BoxName {
   widgetClassPath=&TextBlock
   bIsVariable=true
   parent=ExpandButton
   widget=BoxName
   properties {
    text=Expandable Box
    font {
     bForceMonospaced=false
     bMaterialIsStencil=false
     fontMaterial=null
     fontObject=/Engine/EditorResources/DefaultEditorFont
     letterSpacing=0
     monospacedWidth=1
     size=8.25
     skewAmount=0
     typefaceFontName=Bold
     outlineSettings {
      bApplyOutlineToDropShadows=false
      bMiteredCorners=false
      bSeparateFillAlpha=false
      outlineMaterial=null
      outlineSize=0
      outlineColor=<rgba 0 0 0 1>
     }
    }
    shadowOffset=<0 0>
   }
   slot {class=PanelSlot}
  }
  BoxContent {
   widgetClassPath=&NamedSlot
   bIsVariable=true
   parent=ExpandableBox
   widget=BoxContent
   properties {bExposeOnInstanceOnly=true, visibility=Visible}
   slot {class=VerticalBoxSlot}
  }
  DynamicEntries {widgetClassPath=&VerticalBox, bIsVariable=true, parent=BoxContent, widget=DynamicEntries, slot {class=PanelSlot}}
 ]
}
